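# Animation Blueprint for the MediCrow glider skeleton that applies a fixed additive pose adjustment to an incoming pose.
# An Input Pose passes through chained bone-space Modify Bone nodes that rotate the body offset, both shoulders and both wings, and also shift the shoulders.
# Each node runs at full alpha.
# The AlphaBack and AlphaFwd variables are unused, and the event graph holds only unconnected stub nodes.
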AnimBlueprint /BRCosmetics/Gadgets/Assets/VinderTech_GliderChute/Glider_MediCrow/Meshes/MediCrow_Glider_AnimBP
target_skeleton=&/MediCrow_Glider_Skeleton
graphs [
 AnimGraph {nodes [
   AnimGraphNode_SubInput_0 = Input Pose()
   n1 = "Transform (Modify) Bone - Bone: body_offset"(ComponentPose=Local To Component(LocalPose=@AnimGraphNode_SubInput_0), Alpha=1, Translation="0,0,0", Rotation="0,0,-15", Scale="1,1,1") {
    rotationMode=BMM_Additive
    rotationSpace=BCS_BoneSpace
    alphaBoolBlend {blendOption=Linear}
    alphaScaleBiasClamp {bInterpResult=true, interpSpeedDecreasing=14, interpSpeedIncreasing=14}
    boneToModify {boneName=body_offset}
    rotation {roll=-15}
   }
   n2 = "Transform (Modify) Bone - Bone: shoulder_l"(ComponentPose=@n1, Alpha=1, Translation="0,0,4", Rotation="0,-7,10", Scale="1,1,1") {
    rotationMode=BMM_Additive
    rotationSpace=BCS_BoneSpace
    translationMode=BMM_Additive
    translationSpace=BCS_BoneSpace
    alphaBoolBlend {blendOption=Linear}
    alphaScaleBiasClamp {bInterpResult=true, interpSpeedDecreasing=14, interpSpeedIncreasing=14}
    boneToModify {boneName=shoulder_l}
    rotation {roll=10, yaw=-7}
    translation {z=4}
   }
   n3 = "Transform (Modify) Bone - Bone: shoulder_r"(ComponentPose=@n2, Alpha=1, Translation="0,0,4", Rotation="0,-7,10", Scale="1,1,1") {
    rotationMode=BMM_Additive
    rotationSpace=BCS_BoneSpace
    translationMode=BMM_Additive
    translationSpace=BCS_BoneSpace
    alphaBoolBlend {blendOption=Linear}
    alphaScaleBiasClamp {bInterpResult=true, interpSpeedDecreasing=14, interpSpeedIncreasing=14}
    boneToModify {boneName=shoulder_r}
    rotation {roll=10, yaw=-7}
    translation {z=4}
   }
   n4 = "Transform (Modify) Bone - Bone: allWing_l"(ComponentPose=@n3, Alpha=1, Translation="0,0,0", Rotation="0,-3,0", Scale="1,1,1") {
    rotationMode=BMM_Additive
    rotationSpace=BCS_BoneSpace
    translationMode=BMM_Additive
    translationSpace=BCS_BoneSpace
    alphaBoolBlend {blendOption=Linear}
    alphaScaleBiasClamp {bInterpResult=true, interpSpeedDecreasing=14, interpSpeedIncreasing=14}
    boneToModify {boneName=allWing_l}
    rotation {yaw=-3}
   }
   n5 = "Transform (Modify) Bone - Bone: allWing_r"(ComponentPose=@n4, Alpha=1, Translation="0,0,0", Rotation="0,-3,0", Scale="1,1,1") {
    rotationMode=BMM_Additive
    rotationSpace=BCS_BoneSpace
    translationMode=BMM_Additive
    translationSpace=BCS_BoneSpace
    alphaBoolBlend {blendOption=Linear}
    alphaScaleBiasClamp {bInterpResult=true, interpSpeedDecreasing=14, interpSpeedIncreasing=14}
    boneToModify {boneName=allWing_r}
    rotation {yaw=-3}
   }
   Output Pose(Result=Component To Local(ComponentPose=@n5)) {layerGroup=None}
  ]}
 EventGraph {nodes [
   TryGetPawnOwner()
   Event BlueprintUpdateAnimation()
  ]}
]
variables [
 AlphaBack {type=number}
 AlphaFwd {type=number}
]
